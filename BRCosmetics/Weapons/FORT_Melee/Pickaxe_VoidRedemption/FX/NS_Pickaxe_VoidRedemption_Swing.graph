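# Niagara swing effect for the Void Redemption pickaxe, with variant-based coloring.
# CircleTrail, EmissiveTrail and ShadowTrail are short-lived anim-trail ribbons following the user skeletal mesh; Circles spawns curl-noise-driven sprites by distance moved.
# System variant colors, apparently selected from color arrays, tint particles at spawn, with time-of-day scaling on EmissiveTrail and ShadowTrail.
# EmissiveTrail's update then sets color from one of two gradients by comparing VariantID to zero.
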NiagaraSystem /BRCosmetics/Weapons/FORT_Melee/Pickaxe_VoidRedemption/FX/NS_Pickaxe_VoidRedemption_Swing
dependencies {
 usedDataInterfaces [&NiagaraDataInterfaceCurve, &NiagaraDataInterfaceColorCurve, &NiagaraDataInterfaceArrayColor]
 usedDynamicInputs [&SelectLinearColorFromArray, &MakeIntFromFloat]
 usedModules [
  &SystemState
  &self:SystemSpawnScript.SystemScriptSource.SystemScriptGraph.NiagaraNodeAssignment_0.SetVariables_#1
  &self:SystemSpawnScript.SystemScriptSource.SystemScriptGraph.NiagaraNodeAssignment_1.SetVariables_#2
  &self:SystemSpawnScript.SystemScriptSource.SystemScriptGraph.NiagaraNodeAssignment_2.SetVariables_#3
 ]
}
emitters {
 CircleTrail {
  emitterSpawnScript [&FN_RibbonIDOffsetUtility()]
  emitterUpdateScript [
   &EmitterState()
   &FN_SpawnAnimTrail(AnimTrailWidthMode=::FromFirst)
  ]
  particleSpawnScript [
   &FN_InitializeAnimTrailBoneIndecies()
   &FN_InitializeAnimTrail(AnimTrailWidth=1.4, Color=$System.VariantColor, Lifetime=0.2)
   &AddVelocity(Velocity=<0 400 0>) {enabled=false}
  ]
  particleUpdateScript [
   &SolveForcesAndVelocity(Clamp Velocity=false, Limit Acceleration=false)
   &FN_UpdateAnimTrail(Locked Particle RibbonWidth=1, SkeletalMesh=$User.SkeletalMesh) {module=FN_UpdateAnimTrail001}
   &ScaleColor(
    Linear Color Curve=DataInterface() {propertyValues {
      AlphaCurve {keys [<at 0 1>,<at 0.3479 0.5>,<at 0.8346 0.05>,<at 1 0>]}
      BlueCurve {keys [<at 0 2.9701>,<at 0.4933 0.526>]}
      GreenCurve {keys [<at 0 3>,<at 0.4933 0.4327>]}
      RedCurve {keys [<at 0 2.6548>,<at 0.4933 0.4846>]}
     }}
    Scale Mode=::NewEnumerator2
   )
   &ScaleRibbonWidth(Ribbon Width Scale=&FloatFromCurve(FloatCurve=DataInterface() {propertyValues {Curve {keys [
         {interpMode=RCIM_Cubic, leaveTangent=-0.9299, tangentMode=RCTM_Break, time=0, value=1}
         {arriveTangent=-1.4644, tangentMode=RCTM_User, time=1, value=0}
        ]}}}))
  ]
  rendererClasses [&NiagaraRibbonRendererProperties]
  renderers [
   {rendererClass=&NiagaraRibbonRendererProperties}
   {rendererClass=&NiagaraRibbonRendererProperties, rendererIndex=1}
  ]
 }
 Circles {
  emitterUpdateScript [
   &EmitterState()
   &SpawnPerUnit(Module.Use Max Movement Threshold=true, Module.Use Movement Tolerance=false, Spawn Spacing=40, Use Spawn Probability=false)
  ]
  particleSpawnScript [&InitializeParticle(
    Color=$System.VariantColor
    Lifetime Max=0.6
    Lifetime Min=0.25
    Lifetime Mode=::NewEnumerator1
    Sprite Rotation Mode=::NewEnumerator2
    Sprite Size Mode=::NewEnumerator1
    Uniform Sprite Size Max=&RandomRangeFloat(Minimum=10, Maximum=25)
    Uniform Sprite Size Min=6
    UsePositionOffset=false
    Write Material Random=true
   )]
  particleUpdateScript [
   &ParticleState()
   &ScaleSpriteSize(Uniform Curve Sprite Scale=DataInterface() {propertyValues {Curve {keys [
        {arriveTangent=9.0725, interpMode=RCIM_Cubic, leaveTangent=9.0725, tangentMode=RCTM_User, time=0, value=0}
        {arriveTangent=0.5994, interpMode=RCIM_Cubic, leaveTangent=0.5994, tangentMode=RCTM_User, time=0.1408, value=0.9847}
        {arriveTangent=-2.5503, interpMode=RCIM_Cubic, leaveTangent=-2.5503, tangentMode=RCTM_User, time=1, value=0}
       ]}}})
   &ScaleColor(
    Linear Color Curve=DataInterface() {propertyValues {
      AlphaCurve {keys [<at 0 0.7>,<at 0.4264 0.2>,<at 1 0>]}
      BlueCurve {keys [<at 0 2.4238>,<at 0.4138 1>]}
      GreenCurve {keys [<at 0 2.8137>,<at 0.4138 0.995>]}
      RedCurve {keys [<at 0 3>,<at 0.4138 0.9375>]}
     }}
    Scale Mode=::NewEnumerator2
   )
   &DynamicMaterialParameters(
    Index 0 Param 1=&RandomRangeFloat(Minimum=-1, Maximum=1)
    Index 0 Param 2=&RandomRangeFloat(Minimum=-1, Maximum=1)
    Module.Index0Param0Write=true
    Module.Index0Param1Write=true
    Module.Index0Param2Write=true
    Module.Index0Param3Write=true
   )
   &CurlNoiseForce(Noise Strength=100, Pan Noise=false)
   &SolveForcesAndVelocity(Clamp Velocity=false, Limit Acceleration=false)
  ]
  rendererClasses [&NiagaraSpriteRendererProperties]
  renderers [{rendererClass=&NiagaraSpriteRendererProperties}]
 }
 EmissiveTrail {
  emitterSpawnScript [&FN_RibbonIDOffsetUtility()]
  emitterUpdateScript [
   &EmitterState()
   &FN_SpawnAnimTrail()
  ]
  particleSpawnScript [
   &FN_InitializeAnimTrailBoneIndecies()
   &FN_InitializeAnimTrail(
    AnimTrailWidth=3
    Color=&MultiplyLinearColor(Linear Color=$System.VariantColor, Scale Factor=<1 1 1 0.7>)
    Lifetime=0.2
   )
   &AddVelocity(Velocity=<0 400 0>) {enabled=false}
   &Color(Color=&MultiplyLinearColor(
     Linear Color=$Particles.Color
     Scale Factor=&MakeVector4_FromLinearColor(LinearColor=&FNTimeOfDayColor(Night Color=<rgba 0.4781 0.3563 0.6 0.6>, Morning Color=<rgba 1 0.7537 0.7037 0.5>, Day Color=<rgba 1 1 1 1>, Evening Color=<rgba 1 1 1 1>))
    ))
  ]
  particleUpdateScript [
   &SolveForcesAndVelocity(Clamp Velocity=false, Limit Acceleration=false)
   &FN_UpdateAnimTrail(Locked Particle RibbonWidth=1, SkeletalMesh=$User.SkeletalMesh) {module=FN_UpdateAnimTrail001}
   &ScaleRibbonWidth(Ribbon Width Scale=&FloatFromCurve(FloatCurve=DataInterface() {propertyValues {Curve {keys [
         {arriveTangent=-0.2918, interpMode=RCIM_Cubic, leaveTangent=-1.7404, tangentMode=RCTM_Break, time=0, value=1}
         {arriveTangent=0.0169, tangentMode=RCTM_User, time=1, value=0}
        ]}}}))
   &Color(Color=&MakeCustomLinearColorFromBool(
     Bool=&SetBoolByFloatComparison(A=$User.VariantID, B=0, Comparison Type=::NewEnumerator7)
     True Color=&ColorFromCurve(ColorCurve=DataInterface() {propertyValues {
        AlphaCurve {keys [<at 0 1>,<at 0.3697 0.4>,<at 0.7225 0.1>,<at 1 0>]}
        BlueCurve {keys [<at 0 2.5>,<at 0.2878 1>,<at 0.4243 0.7812>]}
        GreenCurve {keys [<at 0 1.7599>,<at 0.2878 0.6187>,<at 0.4243 0.3975>]}
        RedCurve {keys [<at 0 0.5268>,<at 0.2878 0.1289>,<at 0.4243 0.0402>]}
       }})
     False Color=&ColorFromCurve(ColorCurve=DataInterface() {propertyValues {
        AlphaCurve {keys [<at 0 1>,<at 0.3697 0.4>,<at 0.7225 0.1>,<at 1 0>]}
        BlueCurve {keys [<at 0 1.6922>,<at 0.2878 0.4378>,<at 0.4243 0.4914>]}
        GreenCurve {keys [<at 0 0.5998>,<at 0.2878 0.2158>,<at 0.4243 0.0859>]}
        RedCurve {keys [<at 0 2>,<at 0.2878 1>,<at 0.4243 0.7812>]}
       }})
    )) {module=Color001}
  ]
  rendererClasses [&NiagaraRibbonRendererProperties]
  renderers [
   {rendererClass=&NiagaraRibbonRendererProperties}
   {rendererClass=&NiagaraRibbonRendererProperties, rendererIndex=1}
   {rendererClass=&NiagaraRibbonRendererProperties, rendererIndex=2}
  ]
 }
 ShadowTrail {
  emitterSpawnScript [&FN_RibbonIDOffsetUtility()]
  emitterUpdateScript [
   &EmitterState()
   &FN_SpawnAnimTrail(AnimTrailWidthMode=::FromFirst)
  ]
  particleSpawnScript [
   &FN_InitializeAnimTrailBoneIndecies()
   &FN_InitializeAnimTrail(
    AnimTrailWidth=2
    Color=&MultiplyLinearColor(Linear Color=$System.VariantShadow, Scale Factor=<1 1 1 0.8>)
    Lifetime=0.2
   )
   &Color(Color=&MultiplyLinearColor(
     Linear Color=$Particles.Color
     Scale Factor=&MakeVector4_FromLinearColor(LinearColor=&FNTimeOfDayColor(Night Color=<rgba 0.4317 0.3845 0.6 0.6>, Morning Color=<rgba 1 0.7537 0.7037 0.5>, Day Color=<rgba 1 1 1 1>, Evening Color=<rgba 1 1 1 1>))
    ))
   &AddVelocity(Velocity=<0 400 0>) {enabled=false}
  ]
  particleUpdateScript [
   &SolveForcesAndVelocity(Clamp Velocity=false, Limit Acceleration=false)
   &FN_UpdateAnimTrail(Locked Particle RibbonWidth=1, SkeletalMesh=$User.SkeletalMesh) {module=FN_UpdateAnimTrail001}
   &ScaleColor(
    Linear Color Curve=DataInterface() {propertyValues {
      AlphaCurve {keys [<at 0 1>,<at 0.3172 0.4>,<at 0.6217 0.1>,<at 1 0>]}
      BlueCurve {keys [<at 0.0168 1>]}
      GreenCurve {keys [<at 0.0168 1>]}
      RedCurve {keys [<at 0.0168 1>]}
     }}
    Scale Mode=::NewEnumerator2
   )
   &ScaleRibbonWidth(Ribbon Width Scale=&FloatFromCurve(FloatCurve=DataInterface() {propertyValues {Curve {keys [{interpMode=RCIM_Cubic, leaveTangent=-2.2764, tangentMode=RCTM_Break, time=0, value=1}, <at 1 0>]}}}))
  ]
  rendererClasses [&NiagaraRibbonRendererProperties]
  renderers [
   {rendererClass=&NiagaraRibbonRendererProperties}
   {rendererClass=&NiagaraRibbonRendererProperties, rendererIndex=1}
  ]
 }
}
summary {
 systemName=NS_Pickaxe_VoidRedemption_Swing
 emitters [
  {emitterName=EmissiveTrail, rendererClasses [&NiagaraRibbonRendererProperties]}
  {emitterName=CircleTrail, rendererClasses [&NiagaraRibbonRendererProperties]}
  {emitterName=Circles, rendererClasses [&NiagaraSpriteRendererProperties]}
  {emitterName=ShadowTrail, rendererClasses [&NiagaraRibbonRendererProperties]}
 ]
 userVariables [
  User.RibbonIDOffset {type=NiagaraInt32}
  User.SkeletalMesh {
   type=NiagaraDataInterfaceSkeletalMesh
   defaultValue {
    struct=NiagaraExt_VariableValue_DataInterface
    value {dataInterface=&self:NiagaraDataInterfaceSkeletalMesh_0}
   }
  }
  User.VariantID {type=NiagaraFloat, defaultValue=0, description=""}
 ]
}
